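; Thermal Lance weapon Blueprint that drives the beam, pulse and overheat cosmetics and exposes pulse state to anim Blueprints.
; Simulated proxies use a beam FX component, retrying setup until Owner replicates, and refresh impact FX and audio from the replicated hit result.
; The autonomous proxy takes local trace results from a gameplay ability and interpolates the beam end for small angle changes.
; Overheat values feed Niagara, mesh material, beam audio and local camera shake.
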(event EventBeginPlay
  ;[0 0.15 0.02473 0.5] BeginPlay: Bind to know when the overheat value changes.
  (|CallParentFunction)
  (Overheating|BindEventtoOnOverheatValueChanged (EventDispatchers|CreateEvent :!event HandleOverheatValueChanged))
  (Overheat|HandleOverheatValueChanged (Overheat|GetCurrentOverheatValue) (Overheat|GetOverheatingMaxValue) (Overheat|GetOverheatState)))

(event FortWeapon|EventOnEquippedAndAttached
  ;[0.15 0.07083 0.1402 0.5] OnEquippedAndAttached: Use the niagara beam component as the beam fx.
  ; SimulatedProxies should initialize the beam FX so that it can be picked up by the FortWeaponFxComponent.
  ; AutoProxies will handle their beam FX through the GA_ThermalLance_AutoProxyBeamVisualTrace.
  (|CallParentFunction)
  (Niagara|SetNiagaraVariable(Bool) (Variables|Default|GetNiagara_Beam) "IsLocalPlayer" (Variables|Default|GetbIsLocal))
  (Utilities|FlowControl|Sequence
    (:then_0
      (Beam|InitializeBeamforSimulatedProxyOwner))
    (:then_1
      ;;[0.004387 0.0708 0.15 0.5] If weapon is still overheated, fade in the audio component.
      (if (Overheat|IsOverheated)
        (Audio|Components|Audio|FadeIn (Variables|Default|GetOverheatAudioComponent) 0.1 1.0 0.0 "Linear" :!class /Script/Engine.AudioComponent)))
    (:then_2
      (Rendering|Material|SetCustomPrimitiveDataFloat (Variables|Visual|GetWeaponMesh) 0 0.0))))

(event Overheating|EventOnStartOverheated
  ;[0.1092 0.15 0.02639 0.5] Overheat: Handle overheat fx.
  (|CallParentFunction)
  (Networking|CanExecuteCosmeticEvents
    (:True
      (Components|Activation|Activate (Variables|Default|GetNiagara_Overheat) false)
      (Audio|Components|Audio|FadeIn (Variables|Default|GetOverheatAudioComponent) 0.1 1.0 0.0 "Linear" :!class /Script/Engine.AudioComponent)
      (if (Pawn|IsLocallyViewed (Game|GetInstigator))
        (Audio|PlaySound2D (Variables|Sound|GetOverheatUISoundCue))))))

(event Overheating|EventOnFinishedCooling
  ;[0.1092 0.15 0.02639 0.5] Overheat: Handle overheat fx.
  (|CallParentFunction)
  (Networking|CanExecuteCosmeticEvents
    (:True
      (Components|Activation|Deactivate (Variables|Default|GetNiagara_Overheat))
      (Audio|Components|Audio|FadeOut (Variables|Default|GetOverheatAudioComponent) 0.3 0.0 "Linear" :!class /Script/Engine.AudioComponent))))

(event Custom|SpawnPulseCosmetic
  (bind _niagara_beam (Variables|Default|GetNiagara_Beam))
  ;[0.01749 0.15 0.1325 0.5] SpawnPulseCosmetic: Responsible for spawning the pulse visual on the beam niagara effect.
  (Networking|CanExecuteCosmeticEvents
    (:True
      (Utilities|FlowControl|Sequence
        (:then_0
          (Niagara|SetNiagaraVariable(Bool) _niagara_beam "PulseBeam" true)
          (Niagara|SpawnSystematLocation :SystemTemplate "/ThermalLanceGameplay/Assets/VFX/Systems/NS_ThermalLance_Pulse_Air.NS_ThermalLance_Pulse_Air" :Location (Transformation|GetSocketLocation (Variables|Visual|GetWeaponMesh) "pulseblast") :Scale "1.000000,1.000000,1.000000" :bAutoDestroy true :bAutoActivate true :PoolingMethod "AutoRelease" :bPreCullCheck true)
          (Utilities|FlowControl|RetriggerableDelay 0.4)
          (Niagara|SetNiagaraVariable(Bool) _niagara_beam "PulseBeam" false))
        (:then_1
          (Variables|Transient|SetbPulseCosmeticActive true)
          (Utilities|FlowControl|Delay 0.2)
          (Variables|Transient|SetbPulseCosmeticActive false))))))

(event EventTick
  (params (DeltaSeconds :type Float%20%28single-precision%29))
  ;[0.15 0.05719 0.1428 0.5] Tick: For AutoProxy, interpolate the beam impact location for smoothness.
  (|CallParentFunction DeltaSeconds)
  (Utilities|SwitchNetRole
    (:Autonomous%20Proxy
      (if (Components|Activation|IsActive (Variables|Default|GetNiagara_Beam))
        (Beam|InterpAutoProxyBeamEndLocation DeltaSeconds)))))

(event FortWeapon|EventOnUnEquippedAndDetached
  ;[0.15 0.1339 0.06928 0.5] OnUnEquippedAndDetached: Fade out the audio component.
  (|CallParentFunction)
  (Audio|Components|Audio|FadeOut (Variables|Default|GetOverheatAudioComponent) 0.2 0.0 "Linear" :!class /Script/Engine.AudioComponent)
  (Audio|HandleBeamAudio false)
  (Camera|SetEnableOverheatCameraShake false)
  (Utilities|Time|ClearandInvalidateTimerbyHandle (Variables|Transient|GetRetryInitializeBeamForSimProxyOwnerTimerHandle)))

(event EventPlayWeaponFireCustomAudio
  (params (bPersistentFire :type Boolean) (bSecondaryFire :type Boolean))
  ;[0.01146 0.1047 0.15 0.5] OnPlayWeaponFireFX: Enable the Beam Audio
  (|CallParentFunction bPersistentFire bSecondaryFire)
  (Networking|CanExecuteCosmeticEvents
    (:True
      (Audio|HandleBeamAudio true)
      (Camera|SetEnableOverheatCameraShake true))))

(event EventStopWeaponFireCustomAudio
  ;[0.05388 0.1132 0.15 0.5] OnStopWeaponFireFX: Disable the audio shape line component when the weapon stops firing.
  (|CallParentFunction)
  (Audio|HandleBeamAudio false)
  (Camera|SetEnableOverheatCameraShake false))

(event EventEndPlay
  (params (EndPlayReason :type EEndPlayReason%20Enum :default "Destroyed"))
  ;[0.1042 0.06352 0.009647 0.5] EndPlay: Clean up bound delegates.
  (Utilities|FlowControl|Sequence
    (:then_0
      (Camera|SetEnableOverheatCameraShake false)
      (Overheating|UnbindEventfromOnOverheatValueChanged (EventDispatchers|CreateEvent :!event HandleOverheatValueChanged))
      (Utilities|Time|ClearandInvalidateTimerbyHandle (Variables|Transient|GetRetryInitializeBeamForSimProxyOwnerTimerHandle)))
    (:then_1
      (|CallParentFunction EndPlayReason))))

(event FortWeapon|EventOnPlayImpactFX
  (params (HitResult :type Hit%20Result%20Structure%20%28by%20ref%29) (ImpactPhysicalSurface :type EPhysicalSurface%20Enum :default "SurfaceType_Default") (SpawnedPSC :type FXSystem%20Component%20Object%20Reference))
  (bind _beamtracehitresult (Variables|Transient|GetBeamTraceHitResult))
  (bind (_bblockinghit _binitialoverlap _time _distance _location _impactpoint _normal _impactnormal _physmat _hitactor) (Collision|BreakHitResult _beamtracehitresult))
  ; Check if HitActor is movable. If so, turn off spawned decal actors.
  (if (Utilities|IsValid _hitactor :!class /Script/Engine.KismetSystemLibrary)
    (if (Utilities|Enum|Equal(Enum) (Variables|/Script/Engine.SceneComponent|GetMobility (Variables|/Script/Engine.Actor|GetRootComponent _hitactor)) "Movable")
      (FortWeapon|Effects|SetUseImpactDecals false)
      (else
        (FortWeapon|Effects|SetUseImpactDecals true)))))

(fn GetBeamInfo
  ;[0.15 0.04684 0.1154 0.5] Returns information regarding the beam.
  ; Only makes sense to call the function when the weapon is firing.
  (return (Variables|Transient|GetBeamTraceHitResult)))

(fn GetDataForAnimBP
  (bind (_pulsetiminginfo_pulsestartservertime_4_d0de961546749a2b051b329321ba4e80 _pulsetiminginfo_pulseendservertime_6_337eb0bf48da7bc7dffbd283a317bfed) (Utilities|Struct|BreakThermalLancePulseTimingInfoStruct (Variables|GetPulseTimingInfo_PulseStartServerTime_4_D0DE961546749A2B051B329321BA4E80)))
  (return (Variables|Transient|GetbPulseCosmeticActive) _pulsetiminginfo_pulsestartservertime_4_d0de961546749a2b051b329321ba4e80 _pulsetiminginfo_pulseendservertime_6_337eb0bf48da7bc7dffbd283a317bfed))

(fn GetLatestPulseHitLocation
  (return (Variables|Transient|GetLatestPulseHitImpactLocation)))

(fn GetPulseCosmeticActive
  ;[0.006274 0.1063 0.15 0.5] Returns if the Pulse cosmetic is active, primarily used by anim blueprints.
  (return (Variables|Transient|GetbPulseCosmeticActive)))

(fn GetPulseTimingInfo
  ; Returns the PulseInfoTiming.
  ; Primarily used by the anim blueprints.
  (return (Variables|Transient|GetPulseTimingInfo)))

(fn HandleBeamAudio
  (params (Spawn :type Boolean))
  (if Spawn
    (Variables|Sound|GetBeamAudioActor
      (:then)
      (:else
        (bind _returnvalue (Game|SpawnActorfromClass "/ThermalLanceGameplay/Assets/Sounds/BP_ThermalLance_BeamAudio.BP_ThermalLance_BeamAudio_C" (Math|Transform|MakeTransform (Effects|GetMuzzleLocation -1)) "Undefined" "MultiplyWithRoot" 0 self (Game|GetInstigator)))
        (bind _output_get (Variables|Sound|SetBeamAudioActor _returnvalue))
        (Transformation|AttachActorToActor _output_get self "None" "SnapToTarget" "KeepWorld" "KeepRelative" true)))
    (else
      (bind _beamaudioactor (Variables|Sound|GetBeamAudioActor)
        (:then
          (Actor|DestroyActor _beamaudioactor))
        (:else)))))

(fn HandleOverheatValueChanged
  (params (CurrentValue :type Float%20%28single-precision%29) (MaxValue :type Float%20%28single-precision%29) (State :type EFortWeaponOverheatState%20Enum) (bIsInCooldown :type Boolean))
  ;[0.05702 0.03089 0.15 0.5] HandleOverheatValueChanged: Pass the latest overheat value to the niagara beam.
  (bind _returnvalue (Math|Float|SafeDivide CurrentValue MaxValue))
  (Niagara|SetNiagaraVariable(Float) (Variables|Default|GetNiagara_Beam) "Overheat" _returnvalue)
  (Rendering|Material|SetCustomPrimitiveDataFloat (Variables|Visual|GetWeaponMesh) 0 _returnvalue)
  (bind _camerashakeoverheatcurvefloat (Variables|Camera|GetCameraShakeOverheatCurveFloat)
    (:then
      (bind _overheatcamerashake (Variables|Transient|GetOverheatCameraShake)
        (:then
          (Variables|/Script/Engine.CameraShakeBase|SetShakeScale (Math|Curves|GetFloatValue _camerashakeoverheatcurvefloat _returnvalue) _overheatcamerashake))
        (:else)))
    (:else)))

(fn InitializeBeamForSimulatedProxyOwner
  ;[0.07337 0.03917 0.1406 0.5] InitializeBeamForSimulatedProxyOwner: Initialize the beam so that it can be used by the FortWeaponFxComponent.
  (Utilities|IsValid (Actor|GetOwner)
    (:Is%20Valid
      (Utilities|SwitchNetRole
        (:Simulated%20Proxy
          (Effects|InitializeBeamFX (Variables|Default|GetNiagara_Beam))
          (Utilities|IsValid (Actor|GetComponentbyClass "/Script/FortniteGame.FortWeaponFxComponent_Beam")
            (:Is%20Not%20Valid
              (Game|AddComponentbyClass "/Script/FortniteGame.FortWeaponFxComponent_Beam"))))
        (:Invalid
          ;;[0.4115 0 0 0.5] Warning: SwitchNetRole returned Invalid value.
          (Fortnite|FortShippingLog self "InitializeBeamForSimulatedProxyOwner: SwitchNetRole returned Invalid. Unexpected output." true))) ; Relies on valid owner.
    )
    (:Is%20Not%20Valid
      ;; There's a chance that when the weapon comes into relevancy, the Owner (pawn) has not yet replicated.
      ;; This is especially true when loading into a new area of the map after respawning or when starting to spectate a new player that was previously outside the relevancy range.
      ;; Attempt to initialize the beam in the next frame.
      (bind _returnvalue (Utilities|Time|SetTimerforNextTickbyEvent (EventDispatchers|CreateEvent :!event InitializeBeamForSimulatedProxyOwner)))
      (Variables|Transient|SetRetryInitializeBeamForSimProxyOwnerTimerHandle _returnvalue))))

(fn InterpAutoProxyBeamEndLocation
  (params (DeltaTime :type Float%20%28single-precision%29))
  ;[0.04766 0.1021 0.15 0.5] InterpAutoProxyBeamEndLocation: For autoproxy, Interpolate the beam from the current end location to the target end location if within interpolation threshold.
  (Utilities|SwitchNetRole
    (:Autonomous%20Proxy
      (bind (_returnvalue _anglebetweencurrentandtargetbeam) (Beam|ShouldInterpAutoProxyBeamEndLocation (Variables|Transient|GetCurrentAutoProxyBeamEndLocation) (Variables|Transient|GetTargetAutoProxyBeamEndLocation)))
      (Development|PrintString (Utilities|String|Append "ThermalLanceInterpAngle: " (Utilities|String|ToString(Float) _anglebetweencurrentandtargetbeam)) true false (Math|Color|SelectColor "(R=0.000000,G=0.815780,B=1.000000,A=1.000000)" "(R=1.000000,G=0.000000,B=0.867028,A=1.000000)" _returnvalue) 2.0 "ThermalLanceInterpAngle")
      (bind _output_get (Variables|Transient|SetCurrentAutoProxyBeamEndLocation (select _returnvalue (Math|Interpolation|VInterptoConstant (Variables|Transient|GetCurrentAutoProxyBeamEndLocation) (Variables|Transient|GetTargetAutoProxyBeamEndLocation) DeltaTime (Variables|Settings|GetAutoProxyBeamEndLocationInterpSpeed)) (Variables|Transient|GetTargetAutoProxyBeamEndLocation))))
      (Niagara|SetNiagaraVariable(Position) (Variables|Default|GetNiagara_Beam) "BeamEnd_0" _output_get))))

(fn OnRep_BeamTraceHitResult
  ;[0.15 0.07274 0.1157 0.5] OnRep_BeamTraceHitResult: Refresh the niagara beam based on the latest hit result value.
  (Networking|CanExecuteCosmeticEvents
    (:True
      (Beam|RefreshBeamNiagaraImpactValue)
      (Audio|UpdateAudioShapeLineStartandEndLocations))))

(fn RefreshBeamNiagaraImpactValue
  (bind _beamtracehitresult (Variables|Transient|GetBeamTraceHitResult))
  (bind (_bblockinghit _binitialoverlap _time _distance _location _impactpoint _normal _impactnormal _physmat) (Collision|BreakHitResult _beamtracehitresult))
  (bind _niagara_beam (Variables|Default|GetNiagara_Beam))
  ;[0.08003 0.05506 0.15 0.5] RefreshBeamNiagaraImpactValue: Run for all clients and used to adjust the niagara beam based on the latest hit result value.
  (Networking|CanExecuteCosmeticEvents
    (:True
      (Niagara|SetNiagaraVariable(Bool) _niagara_beam "IsObject" _bblockinghit)
      (Effects|Components|ParticleSystem|SetVectorParameter _niagara_beam "ImpactNormal" _impactnormal)
      (Utilities|IsValid _physmat
        (:Is%20Valid
          (Effects|Components|ParticleSystem|SetBoolParameter _niagara_beam "IsWater" (Utilities|Enum|Equal(Enum) (Variables|/Script/PhysicsCore.PhysicalMaterial|GetSurfaceType _physmat) "SurfaceType8")))
        (:Is%20Not%20Valid
          (Effects|Components|ParticleSystem|SetBoolParameter _niagara_beam "IsWater" false))))))

(fn SetAutoProxyBeamInfo
  (params (BeamHitResult :type Hit%20Result%20Structure))
  ;[0.15 0.02474 0.02474 0.5] SetAutoProxyBeamInfo: AutoProxies handle their own beam visual running local traces.
  ; Called directly from GA_ThermalLance_AutoProxyBeamVisualTrace.
  (Utilities|SwitchNetRole
    (:Autonomous%20Proxy
      (bind _output_get (Variables|Transient|SetBeamTraceHitResult BeamHitResult))
      (Beam|RefreshBeamNiagaraImpactValue)
      (bind _output_get_2 (Variables|Transient|SetTargetAutoProxyBeamEndLocation (select (Collision|BreakHitResult _output_get) (.ImpactPoint (Collision|BreakHitResult _output_get)) (.TraceEnd (Collision|BreakHitResult _output_get)))))
      ;;[0.04262 0.06237 0.15 0.5] Instantly update the end point of the niagara effect if it is not within interpolation threshold.
      (bind _returnvalue (Beam|ShouldInterpAutoProxyBeamEndLocation (Variables|Transient|GetCurrentAutoProxyBeamEndLocation) _output_get_2))
      (if (not _returnvalue)
        (bind _output_get_3 (Variables|Transient|SetCurrentAutoProxyBeamEndLocation (Variables|Transient|GetTargetAutoProxyBeamEndLocation)))
        (Niagara|SetNiagaraVariable(Position) (Variables|Default|GetNiagara_Beam) "BeamEnd_0" _output_get_3)))))

(fn SetAutoProxyBeamNiagaraEnable
  (params (bEnable :type Boolean))
  (bind _niagara_beam (Variables|Default|GetNiagara_Beam))
  ;[0.07217 0.03029 0.15 0.5] SetAutoProxyBeamNiagaraEnable: Activate or deactivate the beam as necessary.
  ; This is directly invoked from GA_ThermalLance_AutoProxyBeamVisualTrace.
  (Utilities|SwitchNetRole
    (:Autonomous%20Proxy
      (if bEnable
        (Components|Activation|Activate _niagara_beam false)
        (else
          (Components|Activation|Deactivate _niagara_beam))))))

(fn SetBeamInfo
  (params (BeamTraceHitResult :type Hit%20Result%20Structure))
  ;[0.0511 0.2309 0.2225 0.5] SetBeamInfo: Update the beam info based on the traces run for the weapon fire.
  ; Called from GA_ThermalLance_PrimaryOnHit.
  (Utilities|SwitchNetRole
    (:Simulated%20Proxy
      _)
    (:Authority
      _))
  (Variables|Transient|SetBeamTraceHitResult BeamTraceHitResult) ; Value is not replicated to Owner since Owner runs its own calculations for best visuals.
)

(fn SetEnableOverheatCameraShake
  (params (bEnable :type Boolean))
  (Networking|CanExecuteCosmeticEvents
    (:True
      (if (Math|Boolean|ANDBoolean bEnable (> (ScalableFloat|GetCachedScalableFloatValue (Variables|Hotfix|GetHotfix_EnableOverheatCameraShake)) 0) (Variables|Default|GetbIsLocal))
        (Variables|Transient|GetOverheatCameraShake
          (:then)
          (:else
            (bind _asfort_player_controller (Utilities|Casting|CastToFortPlayerController (Game|GetInstigatorController)))
            (bind _playercameramanager (Variables|/Script/Engine.PlayerController|GetPlayerCameraManager _asfort_player_controller)
              (:then
                (bind _returnvalue (CameraShakes|StartCameraShake :self _playercameramanager :ShakeClass "/ThermalLanceGameplay/Gameplay/Camera/CameraShake_ThermalLanceOverheat.CameraShake_ThermalLanceOverheat_C" :Scale 1.0 :PlaySpace "CameraLocal"))
                (Variables|Transient|SetOverheatCameraShake _returnvalue))
              (:else))))
        (else
          (bind _overheatcamerashake (Variables|Transient|GetOverheatCameraShake)
            (:then
              (bind _asfort_player_controller_2 (Utilities|Casting|CastToFortPlayerController (Game|GetInstigatorController)))
              (bind _playercameramanager_2 (Variables|/Script/Engine.PlayerController|GetPlayerCameraManager _asfort_player_controller_2)
                (:then
                  (CameraShakes|StopCameraShake _playercameramanager_2 _overheatcamerashake false)
                  (Variables|Transient|SetOverheatCameraShake 0))
                (:else)))
            (:else)))))))

(fn SetLatestPulseHitImpactLocation
  (params (Location :type Vector))
  (Variables|Transient|SetLatestPulseHitImpactLocation Location))

(fn SetPulseTimingInfo
  (params (TimingInfo :type Thermal%20Lance%20Pulse%20Timing%20Info%20Struct%20Structure))
  ; SetPulseTimingInfo: Information is used by te animation blueprint to show feedback on the gauge needle.
  ; Function is called on Server and AutoProxy client.
  (Variables|Transient|SetPulseTimingInfo TimingInfo) ; Variable is not replicated to owner
)

(fn ShouldInterpAutoProxyBeamEndLocation
  (params (CurrentBeamEndLocation :type Vector) (TargetBeamEndLocation :type Vector))
  (bind (_aimdir _returnvalue) (FortWeapon|GetDamageStartLocation -1))
  ;[0.02079 0.1463 0.15 0.5] ShouldInterpAutoProxyBeamEndLocation: Should the end location interp from current to target based on the angle between their directions.
  (bind _returnvalue_2 (Math|Trig|Acos(Degrees) (Math|Vector|DotProduct (Math|Vector|GetUnitDirection(Vector) _returnvalue CurrentBeamEndLocation) (Math|Vector|GetUnitDirection(Vector) _returnvalue TargetBeamEndLocation))))
  (return (< _returnvalue_2 (Variables|Settings|GetAutoProxyBeamEndLocationNonInterpAngle)) _returnvalue_2))

(fn UpdateAudioShapeLineStartAndEndLocations
  (bind (_bblockinghit _binitialoverlap _time _distance _location _impactpoint _normal _impactnormal _physmat _hitactor _hitcomponent _hitbonename _bonename _hititem _elementindex _faceindex _tracestart _traceend) (Collision|BreakHitResult (Variables|Transient|GetBeamTraceHitResult)))
  ;[0.01399 0.1471 0.15 0.5] UpdateAudioShapeLineStartAndEndLocations: Update the start and end points of the audio shape line component.
  (Networking|CanExecuteCosmeticEvents
    (:True
      (bind _beamaudioactor (Variables|Sound|GetBeamAudioActor)
        (:then
          (Class|BPThermalLanceBeamAudio|UpdateBeam _beamaudioactor (Effects|GetMuzzleLocation -1) (select _bblockinghit _impactpoint _traceend) (/ (Overheat|GetCurrentOverheatValue) (Overheat|GetOverheatingMaxValue)) _hitactor))
        (:else)))))

(fn ConstructionScript
  (|CallParentFunction))
