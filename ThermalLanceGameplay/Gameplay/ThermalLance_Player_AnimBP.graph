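# Player animation layer Blueprint for the Thermal Lance, overriding ItemPreAimOffsets with LaserCutter firing additives.
# A fire intro/loop/outro state machine, with clavicle-only mesh-space intro while crouch-moving, feeds an alpha-scaled pullback pose during firing.
# The Fire_Loop additive then layers on by OverheatShake_Alpha, before a final state machine adds a one-shot mesh-space pulse on bDoPulse.
# Unlinking the layer stops the overheat montage on the pawn.
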AnimBlueprint /ThermalLanceGameplay/Gameplay/ThermalLance_Player_AnimBP
target_skeleton=&Fortnite_M_Avg_Player_Skeleton
graphs [
 AnimGraph {nodes [Output Pose()]}
 EventGraph {nodes [
   Event BlueprintOnAnimationLayerUnlinked() {is=title K2Node_Event}
   > StopMontageIfCurrent!(
    execute
    Actor:Actor Object Reference=Get FortPlayerPawn() -> (FortPlayerPawn:Fort Player Pawn Object Reference)
    Montage:Anim Montage Object Reference=/SharedWeaponAnims/Animation/VividRazor/LaserCutter/VividRazor_LaserCutter_Overheat_CMM_M
    OverrideBlendOutTime:float=-1
   ) -> (then)
  ]}
 ItemAimOffset {nodes [
   Input Pose!() -> (Pose, AimOffsetAlpha:double, Yaw:double, Pitch:double) {name=InPose_AimOffset}
   Input Pose!() -> (Pose) {name=InPose_PreAimOffsets}
   Output Pose() {layerGroup=Item Override}
  ]}
 ItemFallAdditive {nodes [
   Input Pose() {name=InPoseItemFallAdditive}
   Output Pose() {layerGroup=Item Override}
  ]}
 ItemFallLandAdditive {nodes [
   Input Pose() {name=InPoseItemFallLandAdditive}
   Output Pose() {layerGroup=Item Override}
  ]}
 ItemFallLevel2Override {nodes [
   Input Pose() {name=InPose_FallLevel2}
   Output Pose() {layerGroup=Item Override}
  ]}
 ItemFinalPoseOverride {nodes [
   Input Pose() {name=InFinalPose}
   Output Pose() {layerGroup=Item Override}
  ]}
 ItemFullBodyAimOffsets {nodes [
   Input Pose!() -> (Pose, FullBody_AimOffsetAlpha:double, Aim_Yaw:double, Aim_Pitch:double) {name=InPoseFullBodyAimOffsets}
   Input Pose!() -> (Pose) {name=InPosePreFullBodyAimOffsets}
   Output Pose() {layerGroup=Item Override}
  ]}
 ItemFullBodyOverride {nodes [
   Input Pose() {name=InPoseFullBody}
   Output Pose() {layerGroup=Item Override}
  ]}
 ItemIdleOverride {nodes [
   Input Pose() {name=InPose_Idle}
   Output Pose() {layerGroup=Item Override}
  ]}
 ItemJetPackJumpAdditive {nodes [Output Pose() {layerGroup=Item Override}]}
 ItemJetPackStartAdditive {nodes [Output Pose() {layerGroup=Item Override}]}
 ItemJumpLoopAdditive {nodes [
   Input Pose() {name=InPoseItemJumpLoopAdditive}
   Output Pose() {layerGroup=Item Override}
  ]}
 ItemJumpOverride {nodes [
   Input Pose() {name=InPoseJump}
   Output Pose() {layerGroup=Item Override}
  ]}
 ItemJumpUpAdditive {nodes [
   Input Pose() {name=InPoseItemJumpUpAdditive}
   Output Pose() {layerGroup=Item Override}
  ]}
 ItemPreAimOffsets {nodes [
   AnimGraphNode_StateMachine_1 = LaserCutter_AdditivePulse() {is=title AnimGraphNode_StateMachine}
   Output Pose(Result=@AnimGraphNode_StateMachine_1) {layerGroup=Item Override}
   Save cached pose 'Input Pose'(Pose=Input Pose() {name=InPosePreAimOffsets}) {cacheName=Input Pose}
   AnimGraphNode_StateMachine_2 = LaserCutter_FullBody_Additive() {is=title AnimGraphNode_StateMachine}
   Save cached pose 'LaserCutter_Fullbody'(Pose=@AnimGraphNode_StateMachine_2) {cacheName=LaserCutter_Fullbody}
   AnimGraphNode_StateMachine_3 = PullbackPoseAdditive() {is=title AnimGraphNode_StateMachine}
   Save cached pose 'LaserCutter_PrePulse'(Pose=Apply Additive(
     Base=@AnimGraphNode_StateMachine_3
     Additive="Sequence Player (additive) 'VividRazor_LaserCutter_Fire_Loop'"() {sequence=/SharedWeaponAnims/Animation/VividRazor/LaserCutter/VividRazor_LaserCutter_Fire_Loop}
     Alpha=Get OverheatShake_Alpha() -> (OverheatShake_Alpha:float)
    ) {alphaBoolBlend {blendOption=Linear}}) {cacheName=LaserCutter_PrePulse}
  ]}
 LaserCutter_AdditivePulse {
  entry=PassThrough
  path=ItemPreAimOffsets.AnimGraphNode_StateMachine_1.LaserCutter_AdditivePulse
  ids=id
  nodes [
   AnimStateNode_0 = PassThrough() {is=title AnimStateNode}
   AnimStateNode_1 = Pulse() {is=title AnimStateNode}
   AnimStateTransitionNode_6 = AnimStateTransitionNode() {blendMode=CubicInOut, crossfadeDuration=0.1}
   AnimStateTransitionNode_7 = AnimStateTransitionNode() {bAutomaticRuleBasedOnSequencePlayerInState=true, blendMode=CircularOut, crossfadeDuration=0.3}
  ]
  links [
   {
    dst=Pulse
    rule=ItemPreAimOffsets.AnimGraphNode_StateMachine_1.LaserCutter_AdditivePulse.AnimStateTransitionNode_6.Transition
    src=PassThrough
    via=AnimStateTransitionNode_6
   }
   {
    dst=PassThrough
    rule=ItemPreAimOffsets.AnimGraphNode_StateMachine_1.LaserCutter_AdditivePulse.AnimStateTransitionNode_7.Transition
    src=Pulse
    via=AnimStateTransitionNode_7
   }
  ]
 }
 PassThrough {
  path=ItemPreAimOffsets.AnimGraphNode_StateMachine_1.LaserCutter_AdditivePulse.AnimStateNode_0.PassThrough
  nodes [Output Animation Pose(Result=Use cached pose 'LaserCutter_PrePulse'())]
 }
 Pulse {
  path=ItemPreAimOffsets.AnimGraphNode_StateMachine_1.LaserCutter_AdditivePulse.AnimStateNode_1.Pulse
  nodes [Output Animation Pose(Result=Apply Mesh Space Additive!(
     Base:Pose Link Structure=Use cached pose 'LaserCutter_PrePulse'()
     Additive:Pose Link Structure="Sequence Player (additive) 'VividRazor_LaserCutter_Pulse_MeshSpace'"() {bLoopAnimation=false, sequence=/SharedWeaponAnims/Animation/VividRazor/LaserCutter/VividRazor_LaserCutter_Pulse_MeshSpace}
     Alpha:float=1
    ) -> (Pose:Pose Link Structure) {is=title AnimGraphNode_ApplyMeshSpaceAdditive, alphaBoolBlend {blendOption=Linear}})]
 }
 Transition {
  path=ItemPreAimOffsets.AnimGraphNode_StateMachine_1.LaserCutter_AdditivePulse.AnimStateTransitionNode_6.Transition
  nodes [Result(bCanEnterTransition=Get bDoPulse() -> (bDoPulse:bool))]
 }
 Transition {path=ItemPreAimOffsets.AnimGraphNode_StateMachine_1.LaserCutter_AdditivePulse.AnimStateTransitionNode_7.Transition, nodes [Result()]}
 LaserCutter_FullBody_Additive {
  entry=Passthrough
  path=ItemPreAimOffsets.AnimGraphNode_StateMachine_2.LaserCutter_FullBody_Additive
  ids=id
  nodes [
   AnimStateNode_0 = Passthrough() {is=title AnimStateNode}
   AnimStateNode_1 = Fire In() {is=title AnimStateNode}
   AnimStateNode_2 = Fire Loop() {is=title AnimStateNode}
   AnimStateNode_3 = Fire End() {is=title AnimStateNode}
   AnimStateTransitionNode_0 = AnimStateTransitionNode() {crossfadeDuration=0}
   AnimStateTransitionNode_1 = AnimStateTransitionNode() {bAutomaticRuleBasedOnSequencePlayerInState=true}
   AnimStateTransitionNode_2 = AnimStateTransitionNode()
   AnimStateTransitionNode_3 = AnimStateTransitionNode() {bAutomaticRuleBasedOnSequencePlayerInState=true}
   AnimStateTransitionNode_4 = AnimStateTransitionNode()
   AnimStateTransitionNode_5 = AnimStateTransitionNode()
  ]
  links [
   {
    dst=Fire In
    rule=ItemPreAimOffsets.AnimGraphNode_StateMachine_2.LaserCutter_FullBody_Additive.AnimStateTransitionNode_0.Transition
    src=Fire End
    via=AnimStateTransitionNode_5
   }
   {
    dst=Passthrough
    rule=ItemPreAimOffsets.AnimGraphNode_StateMachine_2.LaserCutter_FullBody_Additive.AnimStateTransitionNode_3.Transition
    src=Fire End
    via=AnimStateTransitionNode_3
   }
   {
    dst=Fire End
    rule=ItemPreAimOffsets.AnimGraphNode_StateMachine_2.LaserCutter_FullBody_Additive.AnimStateTransitionNode_4.Transition
    src=Fire In
    via=AnimStateTransitionNode_4
   }
   {
    dst=Fire Loop
    rule=ItemPreAimOffsets.AnimGraphNode_StateMachine_2.LaserCutter_FullBody_Additive.AnimStateTransitionNode_1.Transition
    src=Fire In
    via=AnimStateTransitionNode_1
   }
   {
    dst=Fire End
    rule=ItemPreAimOffsets.AnimGraphNode_StateMachine_2.LaserCutter_FullBody_Additive.AnimStateTransitionNode_4.Transition
    src=Fire Loop
    via=AnimStateTransitionNode_2
   }
   {
    dst=Fire In
    rule=ItemPreAimOffsets.AnimGraphNode_StateMachine_2.LaserCutter_FullBody_Additive.AnimStateTransitionNode_0.Transition
    src=Passthrough
    via=AnimStateTransitionNode_0
   }
  ]
 }
 Passthrough {
  path=ItemPreAimOffsets.AnimGraphNode_StateMachine_2.LaserCutter_FullBody_Additive.AnimStateNode_0.Passthrough
  nodes [Output Animation Pose(Result=Use cached pose 'Input Pose'())]
 }
 Fire In {
  path=ItemPreAimOffsets.AnimGraphNode_StateMachine_2.LaserCutter_FullBody_Additive.AnimStateNode_1.Fire In
  nodes [
   n1 = "Sequence Player (additive) 'VividRazor_LaserCutter_Fire_Intro'"() {bLoopAnimation=false, sequence=/SharedWeaponAnims/Animation/VividRazor/LaserCutter/VividRazor_LaserCutter_Fire_Intro}
   n2 = Use cached pose 'Input Pose'()
   n3 = Layered blend per bone(
    BasePose=Use cached pose 'Input Pose'()
    BlendPoses_0=Apply Mesh Space Additive(
     Base=Use cached pose 'Input Pose'()
     Additive="Sequence Player (additive) 'VividRazor_LaserCutter_Fire_Intro_MeshSpace'"() {bLoopAnimation=false, sequence=/SharedWeaponAnims/Animation/VividRazor/LaserCutter/VividRazor_LaserCutter_Fire_Intro_MeshSpace}
     Alpha=1
    ) {alphaBoolBlend {blendOption=Linear}}
   ) {bMeshSpaceRotationBlend=true, layerSetup [{branchFilters [{boneName=clavicle_l}, {boneName=clavicle_r}]}]}
   n4 = Get bIsCrouchMoving() -> (bIsCrouchMoving:bool)
   Output Animation Pose(Result=Blend Poses by bool(BlendPose_0=Apply Additive(Base=@n2, Additive=@n1) {alphaBoolBlend {blendOption=Linear}}, BlendPose_1=@n3, bActiveValue=NOT Boolean(A=@n4)) {blendType=Linear})
  ]
 }
 Fire Loop {
  path=ItemPreAimOffsets.AnimGraphNode_StateMachine_2.LaserCutter_FullBody_Additive.AnimStateNode_2.Fire Loop
  nodes [Output Animation Pose(Result=Apply Additive(
     Base=Use cached pose 'Input Pose'()
     Additive="Sequence Player (additive) 'VividRazor_LaserCutter_Fire_Loop'"() {playRate=0, sequence=&VividRazor_LaserCutter_Fire_Loop}
    ) {alphaBoolBlend {blendOption=Linear}})]
 }
 Fire End {
  path=ItemPreAimOffsets.AnimGraphNode_StateMachine_2.LaserCutter_FullBody_Additive.AnimStateNode_3.Fire End
  nodes [Output Animation Pose(Result=Apply Additive(
     Base=Use cached pose 'Input Pose'()
     Additive="Sequence Player (additive) 'VividRazor_LaserCutter_Fire_Outro'"() {bLoopAnimation=false, sequence=/SharedWeaponAnims/Animation/VividRazor/LaserCutter/VividRazor_LaserCutter_Fire_Outro}
    ) {alphaBoolBlend {blendOption=Linear}})]
 }
 Transition {
  path=ItemPreAimOffsets.AnimGraphNode_StateMachine_2.LaserCutter_FullBody_Additive.AnimStateTransitionNode_0.Transition
  nodes [Result(bCanEnterTransition=Get bIsFiring() -> (bIsFiring:bool))]
 }
 Transition {path=ItemPreAimOffsets.AnimGraphNode_StateMachine_2.LaserCutter_FullBody_Additive.AnimStateTransitionNode_1.Transition, nodes [Result()]}
 Transition {path=ItemPreAimOffsets.AnimGraphNode_StateMachine_2.LaserCutter_FullBody_Additive.AnimStateTransitionNode_3.Transition, nodes [Result()]}
 Transition {
  path=ItemPreAimOffsets.AnimGraphNode_StateMachine_2.LaserCutter_FullBody_Additive.AnimStateTransitionNode_4.Transition
  nodes [Result(bCanEnterTransition=NOT Boolean(A=Get bIsFiring()))]
 }
 PullbackPoseAdditive {
  entry=Fullbody_Pose
  path=ItemPreAimOffsets.AnimGraphNode_StateMachine_3.PullbackPoseAdditive
  ids=id
  nodes [
   AnimStateNode_0 = Fullbody_Pose() {is=title AnimStateNode}
   AnimStateNode_1 = Additive_PullbackPose() {is=title AnimStateNode}
   AnimStateTransitionNode_0 = AnimStateTransitionNode() {blendMode=QuarticInOut, crossfadeDuration=0.5}
   AnimStateTransitionNode_1 = AnimStateTransitionNode() {blendMode=CubicInOut, crossfadeDuration=0.5}
  ]
  links [
   {
    dst=Fullbody_Pose
    rule=ItemPreAimOffsets.AnimGraphNode_StateMachine_3.PullbackPoseAdditive.AnimStateTransitionNode_1.Transition
    src=Additive_PullbackPose
    via=AnimStateTransitionNode_1
   }
   {
    dst=Additive_PullbackPose
    rule=ItemPreAimOffsets.AnimGraphNode_StateMachine_3.PullbackPoseAdditive.AnimStateTransitionNode_0.Transition
    src=Fullbody_Pose
    via=AnimStateTransitionNode_0
   }
  ]
 }
 Fullbody_Pose {
  path=ItemPreAimOffsets.AnimGraphNode_StateMachine_3.PullbackPoseAdditive.AnimStateNode_0.Fullbody_Pose
  nodes [Output Animation Pose(Result=Use cached pose 'LaserCutter_Fullbody'())]
 }
 Additive_PullbackPose {
  path=ItemPreAimOffsets.AnimGraphNode_StateMachine_3.PullbackPoseAdditive.AnimStateNode_1.Additive_PullbackPose
  nodes [
   n1 = "Sequence Player (additive) 'VividRazor_LaserCutter_Pose_Fire_Extreme_Back_CMM_MeshSpace'"() {
    bLoopAnimation=false
    sequence=/SharedWeaponAnims/Animation/VividRazor/LaserCutter/VividRazor_LaserCutter_Pose_Fire_Extreme_Back_CMM_MeshSpace
   }
   Output Animation Pose(Result=Apply Mesh Space Additive(Base=Use cached pose 'LaserCutter_Fullbody'(), Additive=@n1, Alpha=Get AdditiveAlpha() -> (AdditiveAlpha:float)) {alphaBoolBlend {blendOption=Linear}})
  ]
 }
 Transition {
  path=ItemPreAimOffsets.AnimGraphNode_StateMachine_3.PullbackPoseAdditive.AnimStateTransitionNode_0.Transition
  nodes [Result(bCanEnterTransition=Get bIsFiring())]
 }
 Transition {
  path=ItemPreAimOffsets.AnimGraphNode_StateMachine_3.PullbackPoseAdditive.AnimStateTransitionNode_1.Transition
  nodes [Result(bCanEnterTransition=NOT Boolean(A=Get bIsFiring()))]
 }
 ItemPreIK {nodes [
   Input Pose() {name=InPose_PreIK}
   Output Pose() {layerGroup=Item Override}
  ]}
 ItemRightHandIK {nodes [
   Input Pose() {name=InPosePreRightHandIK}
   Input Pose() {name=InPosePostRightHandIK}
   Output Pose() {layerGroup=Item Override}
  ]}
 ItemSkeletalControl {nodes [
   Input Pose()
   Output Pose() {layerGroup=Item Override}
  ]}
 ItemSwimJumpFallLoop {nodes [Output Pose() {layerGroup=Item Override}]}
 ItemSwimJumpSurfaceEnd {nodes [Output Pose() {layerGroup=Item Override}]}
 ItemSwimJumpSurfaceLoop {nodes [Output Pose() {layerGroup=Item Override}]}
 ItemUpperBody {nodes [
   Input Pose!() -> (Pose, InputParam:Fort Anim Input Adjusted Aim Structure) {name=InPoseUpperBody}
   Output Pose() {layerGroup=Item Override}
  ]}
 ItemZipLineStartAdditive {nodes [Output Pose() {layerGroup=Item Override}]}
]
variables []
